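# Default Lit material for the Pasta Ripper rifle's magazine effect, flagged for use on Niagara mesh particles.
# It samples five of the rifle's own textures, with suffixes D, N, S, CM and E, reading S and CM as masks and N as a normal map.
# The graph holds only these texture samples, with no parameters or math nodes and no visible connections to material outputs.
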Material /Game/Effects/Fort_Effects/Effects/Weapon/M_FX_PastaRipper_Mag
graph {
 properties {bUsedWithNiagaraMeshParticles=true, pixelDepthOffsetMode=PDOM_Legacy, usedShadingModels=Default Lit}
 nodes [
  TextureSample() {samplerType=SAMPLERTYPE_Masks, texture=/Game/Weapons/FORT_Rifles/Textures/T_Pasta_Ripper_S}
  TextureSample() {samplerType=SAMPLERTYPE_Masks, texture=/Game/Weapons/FORT_Rifles/Textures/T_Pasta_Ripper_CM}
  TextureSample() {texture=/Game/Weapons/FORT_Rifles/Textures/T_Pasta_Ripper_D}
  TextureSample() {samplerType=SAMPLERTYPE_Normal, texture=/Game/Weapons/FORT_Rifles/Textures/T_Pasta_Ripper_N}
  TextureSample() {texture=/Game/Weapons/FORT_Rifles/Textures/T_Pasta_Ripper_E}
 ]
}
